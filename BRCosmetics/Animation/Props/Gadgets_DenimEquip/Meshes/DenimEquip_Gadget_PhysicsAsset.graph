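# Physics asset for the DenimEquip gadget cosmetic prop, giving its skeletal mesh capsule collision bodies.
# Capsules cover the body, head, tail, two wings, four miniGuff joints, a main rope joint and three thin rope segments.
# Constraints with uniform swing and twist limit angles form two separate groups: rope segments to the main rope joint, and wings to the body.
# The head, tail and miniGuff bodies have no constraints.
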PhysicsAsset /BRCosmetics/Animation/Props/Gadgets_DenimEquip/Meshes/DenimEquip_Gadget_PhysicsAsset
bodies {
 JNT_B_Rope_02 [{shapeType=Capsule, length=29.1875, radius=0.604, center=<7.2072 0 -0.0804>, rotation=<rot 0 89.9751 -89.9963>}]
 JNT_C_Body [{shapeType=Capsule, length=33.6611, radius=33.4003, center=<-0.0023 2.2923 -4.3229>, rotation=<rot 65.5449 -90.0166 -0.0151>}]
 JNT_C_Rope_02 [{shapeType=Capsule, length=33.2541, radius=0.4914, center=<8.2104 0 0.2093>, rotation=<rot 0 -90.0164 -89.7826>}]
 JNT_D_Rope_02 [{shapeType=Capsule, length=31.2461, radius=0.5212, center=<7.7224 0 0.1941>, rotation=<rot 0 -89.9789 -89.817>}]
 JNT_Ropes_Main [{shapeType=Capsule, length=37.9489, radius=26.8771, center=<-5.5969 -46.1468 -3.8795>, rotation=<rot 0 85.8274 -87.2809>}]
 JNT_Wing_l [{shapeType=Capsule, length=12.3644, radius=10.19, center=<-12.6415 -0.2368 -0.451>, rotation=<rot 0 -122.6193 -84.0543>}]
 JNT_Wing_r [{shapeType=Capsule, length=12.3644, radius=10.19, center=<12.6353 0.2376 0.4585>, rotation=<rot 0 -122.6194 -84.0543>}]
 JNT_mainHead [{shapeType=Capsule, length=29.1872, radius=28.7847, center=<16.7468 -0.0007 7.0777>, rotation=<rot 77.2445 179.9509 -90.0503>}]
 JNT_mainTail [{shapeType=Capsule, length=22.7409, radius=17.3816, center=<17.7887 -0.0621 5.3697>, rotation=<rot 62.6817 -175.5255 -84.9665>}]
 JNT_miniGuff_01 [{shapeType=Capsule, length=9.5752, radius=9.3369, center=<-0.5403 0.4721 0.3312>, rotation=<rot 6.3425 -90 -83.5699>}]
 JNT_miniGuff_02 [{shapeType=Capsule, length=14.7956, radius=14.3979, center=<-1.0201 0.0457 0.5062>, rotation=<rot 6.3924 -90 -83.7291>}]
 JNT_miniGuff_03 [{shapeType=Capsule, length=16.8615, radius=16.4083, center=<-1.8082 -0.006 0.9789>, rotation=<rot 6.3924 -90 -83.7291>}]
 JNT_miniGuff_04 [{shapeType=Capsule, length=12.721, radius=12.3791, center=<-1.4511 -0.7386 0.2174>, rotation=<rot 6.3924 -90 -83.7291>}]
}
constraints [
 {bone1Name=JNT_D_Rope_02, bone2Name=JNT_Ropes_Main, swing1LimitDegrees=45, swing2LimitDegrees=45, twistLimitDegrees=45}
 {bone1Name=JNT_C_Rope_02, bone2Name=JNT_Ropes_Main, swing1LimitDegrees=45, swing2LimitDegrees=45, twistLimitDegrees=45}
 {bone1Name=JNT_B_Rope_02, bone2Name=JNT_Ropes_Main, swing1LimitDegrees=45, swing2LimitDegrees=45, twistLimitDegrees=45}
 {bone1Name=JNT_Wing_l, bone2Name=JNT_C_Body, swing1LimitDegrees=45, swing2LimitDegrees=45, twistLimitDegrees=45}
 {bone1Name=JNT_Wing_r, bone2Name=JNT_C_Body, swing1LimitDegrees=45, swing2LimitDegrees=45, twistLimitDegrees=45}
]
